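# Physics asset for the Teddy Bear glider skeletal mesh, giving its parachute and bear parts physics bodies for constrained secondary motion.
# Five single-capsule bodies cover the parachute, head, arms (one shared body) and each leg.
# The parachute is the hub, with every bear part constrained directly to it.
# Tight angular limits let head and legs swing on one axis and twist slightly, while the arms only twist.
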PhysicsAsset /BRCosmetics/Gadgets/Assets/VinderTech_GliderChute/Glider_TeddyBear/Meshes/SK_Glider_TeddyBear_PhysicsAsset
bodies {
 bearArms [{shapeType=Capsule, length=56.6823, radius=33.0815, center=<0.5659 24.0082 0.2261>, rotation=<rot 0 90.0282 -89.9644>}]
 bearHead [{shapeType=Capsule, length=62.6867, radius=50.1495, center=<-1.0539 2.8488 53.4179>, rotation=<rot -10.425 -88.4812 -89.7856>}]
 bearLeg_L [{shapeType=Capsule, length=33.0611, radius=21.0082, center=<0.9766 32.4833 -1.7512>, rotation=<rot 0 0.0838 -88.0409>}]
 bearLeg_R [{shapeType=Capsule, length=33.064, radius=21.0147, center=<-1.0461 32.5111 -1.4267>, rotation=<rot 0 -0.5063 -88.2035>}]
 parachute [{shapeType=Capsule, length=58.5537, radius=48.6188, center=<0.002 22.908 44.99>, rotation=<rot -33.1784 -90 -0.0169>}]
}
constraints [
 {bone1Name=bearLeg_R, bone2Name=parachute, swing1LimitDegrees=10, swing2LimitDegrees=0, twistLimitDegrees=10}
 {bone1Name=bearLeg_L, bone2Name=parachute, swing1LimitDegrees=10, swing2LimitDegrees=0, twistLimitDegrees=10}
 {bone1Name=bearArms, bone2Name=parachute, swing1LimitDegrees=0, swing2LimitDegrees=0, twistLimitDegrees=20}
 {bone1Name=bearHead, bone2Name=parachute, swing1LimitDegrees=10, swing2LimitDegrees=0, twistLimitDegrees=10}
]
